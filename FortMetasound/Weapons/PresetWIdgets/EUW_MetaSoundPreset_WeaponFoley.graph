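# Editor utility widget that gives a custom preset editor panel for weapon foley MetaSound sources (ADS, canting, equip, reload).
# On construction it stores the preset builder and parent MetaSound, then initializes MetaSound editor viewmodels for the preset and parent.
# It binds a Master Volume slider and Random Pitch base and spread knobs to input viewmodels looked up by each widget's input names through the literal widget interface.
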EditorUtilityWidgetBlueprint /FortMetasound/Weapons/PresetWIdgets/EUW_MetaSoundPreset_WeaponFoley
graphs [
 EventGraph {nodes [
   Initialize Literal Widgets() {is=title K2Node_CustomEvent}
   > n1 = For Each Loop(Array:Array of Meta Sound Literal Widget Interface Interfaces=Make Array!(
     "[0]":Meta Sound Literal Widget Interface Interface=Get MasterVolume_Slider() -> (MasterVolume_Slider:EUW Meta Sound Float Slider Default Vertical Object Reference)
     "[1]":Meta Sound Literal Widget Interface Interface=Get RandomPitchBase_Knob() -> (RandomPitchBase_Knob:Meta Sound Literal Float Knob Default Object Reference)
     "[2]":Meta Sound Literal Widget Interface Interface=Get RandomPitchSpread_Knob() -> (RandomPitchSpread_Knob:Meta Sound Literal Float Knob Default Object Reference)
    ) -> (Array:Array of Meta Sound Literal Widget Interface Interfaces)) -> (Array Element:Meta Sound Literal Widget Interface Interface)
   > Set Input Viewmodels!(
    execute
    self:Meta Sound Literal Widget Interface Interface=@n1
    InputViewModels:Map of Names to Meta Sound Input Viewmodel Object References=Get Input Viewmodels!(
     Input Names:Array of Names=Get Input Viewmodel Names!(self:Meta Sound Literal Widget Interface Interface=@n1) -> (ReturnValue:Array of Names) {is=title K2Node_CallFunction}
     MetaSound Viewmodel:Meta Sound Viewmodel Object Reference=Get MetaSoundEditorViewModel() -> (MetaSoundEditorViewModel:Meta Sound Editor Viewmodel Object Reference)
    ) -> (Input Viewmodels:Map of Names to Meta Sound Input Viewmodel Object References) {is=title K2Node_CallFunction}
   ) -> (then) {is=title K2Node_CallFunction}
   Event Construct()
   > n2 = Get!(execute:Exec) -> (then:Exec, else:Exec, Builder:Meta Sound Builder Base Object Reference)
   > Initialize Builder!(execute, self:Meta Sound Viewmodel Object Reference=Get MetaSoundEditorViewModel(), InBuilder:Meta Sound Builder Base Object Reference=@n2) -> (then)
   > Initialize MetaSound!(
    execute
    self:Meta Sound Viewmodel Object Reference=Get ParentMetaSoundEditorViewModel() -> (ParentMetaSoundEditorViewModel:Meta Sound Editor Viewmodel Object Reference)
    InMetaSound:Meta Sound Document Interface Interface=Get Parent MetaSound() -> (Parent MetaSound:Meta Sound Source Object Reference)
   ) -> (then)
   > Initialize Literal Widgets() {is=title K2Node_CallFunction}
   n3 = Event On MetaSoundPreset Widget Constructed!() -> (OutputDelegate, then, Builder:Meta Sound Builder Base Object Reference) {is=title K2Node_Event}
   > n4 = Set Builder(Builder:Meta Sound Builder Base Object Reference=@n3.Builder) -> (Output_Get:Meta Sound Builder Base Object Reference)
   > n5 = Cast To MetaSoundSource!(
    execute
    Object=Get Preset Parent!(self:Meta Sound Builder Base Object Reference=@n4) -> (ReturnValue:Object Reference) {is=title K2Node_CallFunction}
   ) -> (then, CastFailed, AsMetaSound Source:Meta Sound Source Object Reference) {is=title K2Node_DynamicCast}
   > Set Parent MetaSound(Parent MetaSound:Meta Sound Source Object Reference=@n5) -> (Output_Get:Meta Sound Source Object Reference)
  ]}
 Get Input Viewmodels {nodes [
   n1 = Get Input Viewmodels!() -> (then, Input Names:Array of Names, MetaSound Viewmodel:Meta Sound Viewmodel Object Reference) {is=title K2Node_FunctionEntry}
   > n2 = For Each Loop(Array:Array of Names=@n1) -> (Array Element:Name)
   Return Node!(
    execute=@n2.Completed
    Input Viewmodels:Map of Names to Meta Sound Input Viewmodel Object References=Get Viewmodels() -> (Viewmodels:Map of Names to Meta Sound Input Viewmodel Object References)
   )
   > Add!(
    execute
    TargetMap:Map of Names to Meta Sound Input Viewmodel Object References=Get Viewmodels()
    Key:"Name (by ref)"=@n2
    Value:"Meta Sound Input Viewmodel Object Reference (by ref)"=Find Input Viewmodel!(self:Meta Sound Viewmodel Object Reference=@n1.MetaSound Viewmodel, InputViewModelName:Name=@n2) -> (ReturnValue:Meta Sound Input Viewmodel Object Reference)
   ) -> (then) {is=title K2Node_CallFunction}
  ]}
 GetEditorName {nodes [
   GetEditorName() {is=title K2Node_FunctionEntry}
   > Return Node!(execute, ReturnValue:Text)
  ]}
 GetIconBrushName {nodes [
   GetIconBrushName() {is=title K2Node_FunctionEntry}
   > Return Node(ReturnValue:Name=None)
  ]}
 GetSupportedMetaSounds {nodes [
   GetSupportedMetaSounds() {is=title K2Node_FunctionEntry}
   > Return Node!(
    execute
    bSupportAllPresets:bool=false
    ExcludedMetaSounds:Array of Meta Sound Document Interface Interfaces
    IncludedMetaSounds:Array of Meta Sound Document Interface Interfaces=Get Supported MetaSounds() -> (Supported MetaSounds:Array of Meta Sound Document Interface Interfaces)
   )
  ]}
 OnInitializingTab {nodes [
   On Initializing Tab() {is=title K2Node_FunctionEntry}
   > Return Node!(execute, TabInfo:Meta Sound Template Widget Tab Info Structure)
  ]}
]
variables [
 Builder {type=MetaSoundBuilderBase, default=null}
 Parent MetaSound {type=MetaSoundSource, default=null, class=~, properties=~}
 Supported MetaSounds {
  type="array<object>"
  default [
   MetaSoundSource'/FortMetasound/Weapons/MSS_Weapon_ADS_In'
   MetaSoundSource'/FortMetasound/Weapons/MSS_Weapon_ADS_Out'
   MetaSoundSource'/FortMetasound/Weapons/MSS_Weapon_Canting_FirstPerson'
   MetaSoundSource'/FortMetasound/Weapons/MSS_Weapon_Canting_ThirdPerson'
   MetaSoundSource'/FortMetasound/Weapons/MSS_Weapon_Equip'
   MetaSoundSource'/FortMetasound/Weapons/MSS_Weapon_Reload_Cartridge_Eject'
   MetaSoundSource'/FortMetasound/Weapons/MSS_Weapon_Reload_Cartridge_Insert'
   MetaSoundSource'/FortMetasound/Weapons/MSS_Weapon_Reload_End'
   MetaSoundSource'/FortMetasound/Weapons/MSS_Weapon_Reload_Insert'
   MetaSoundSource'/FortMetasound/Weapons/MSS_Weapon_Reload_Start'
  ]
 }
]
widgets {
 info {rootWidgetClass=&Border, widgetCount=14}
 widgets [
  Border_1 {widgetClassPath=&Border, widget=Border_1, properties {brushColor=<rgba 0.0103 0.0103 0.0103 1>, padding=<ltrb 0 0 0 0>}}
  VerticalBox_0 {widgetClassPath=&VerticalBox, parent=Border_1, widget=VerticalBox_0, slot {class=BorderSlot, properties {padding=<ltrb 0 0 0 0>}}}
  CommonTextBlock_359 {
   widgetClassPath=&CommonTextBlock
   parent=VerticalBox_0
   widget=CommonTextBlock_359
   properties {
    text=Parent Display Name
    colorAndOpacity {colorUseRule=UseColor_Specified, specifiedColor=<rgba 0.5271 0.5271 0.5271 1>}
    font {
     bForceMonospaced=false
     bMaterialIsStencil=false
     fontMaterial=null
     fontObject=&Roboto
     letterSpacing=0
     monospacedWidth=1
     size=18
     skewAmount=0
     typefaceFontName=Bold
     outlineSettings {
      bApplyOutlineToDropShadows=false
      bMiteredCorners=false
      bSeparateFillAlpha=false
      outlineMaterial=null
      outlineSize=0
      outlineColor=<rgba 0 0 0 1>
     }
    }
    shadowOffset=<0 0>
   }
   slot {class=VerticalBoxSlot, properties {horizontalAlignment=HAlign_Left, padding=<ltrb 16 16 16 8>}}
  }
  LiteralWidgets_Box {
   widgetClassPath=&WrapBox
   bIsVariable=true
   parent=VerticalBox_0
   widget=LiteralWidgets_Box
   properties {bOverride_Cursor=true, innerSlotPadding=<16 0>}
   slot {class=VerticalBoxSlot, properties {padding=<ltrb 16 8 16 16>}}
  }
  Border {
   widgetClassPath=&Border
   parent=LiteralWidgets_Box
   widget=Border
   properties {
    background {
     bIsDynamicallyLoaded=false
     drawAs=RoundedBox
     imageType=NoImage
     mirroring=NoMirror
     resourceName=None
     resourceObject=null
     tiling=NoTile
     imageSize=<32 32>
     margin=<ltrb 0 0 0 0>
     outlineSettings {
      bUseBrushTransparency=false
      roundingType=FixedRadius
      width=0
      color {colorUseRule=UseColor_Specified, specifiedColor=<rgba 0.0075 0.0075 0.0075 1>}
      cornerRadii=<12 12 12 12>
     }
     tintColor {colorUseRule=UseColor_Specified, specifiedColor=<rgba 1 1 1 1>}
     uVRegion {bIsValid=false, max=<0 0>, min=<0 0>}
    }
    brushColor=<rgba 0.0212 0.0212 0.0212 1>
    padding=<ltrb 8 8 8 8>
   }
   slot {class=WrapBoxSlot, properties {verticalAlignment=VAlign_Top}}
  }
  VerticalBox {widgetClassPath=&VerticalBox, parent=Border, widget=VerticalBox, slot {class=BorderSlot, properties {padding=<ltrb 8 8 8 8>}}}
  MasterVolume_Slider {
   widgetClassPath=/TechAudioToolsContent/Widgets/MetaSoundLiterals/Float/EUW_MetaSoundFloat_Slider_DefaultVertical
   bIsVariable=true
   parent=VerticalBox
   widget=MasterVolume_Slider
   properties {
    apply Logarithmic Interpolation=true
    input Name=MasterVolume
    label Text=Master Volume
    metaSoundLiteralViewModel_Float=&self:WidgetTree.MasterVolume_Slider.MetaSoundLiteralViewModel_Float_0
    width Override=96
    slider Style {
     material=null
     desiredSize=<25 250>
     sliderBackgroundAccentColor=<rgba 0.005 0.005 0.005 1>
     sliderBackgroundColor=<rgba 0.008 0.008 0.008 1>
     sliderHandleMainColor=<rgba 1 1 1 1>
     sliderHandleOutlineColor=<rgba 0.15 0.15 0.15 1>
     sliderValueMainColor=<rgba 0 1 0.0698 1>
     textBoxStyle {
      backgroundColor {colorUseRule=UseColor_ColorTable, specifiedColor=<rgba 0 0 0 0>}
      backgroundImage {
       bIsDynamicallyLoaded=false
       drawAs=RoundedBox
       imageType=NoImage
       mirroring=NoMirror
       resourceName=None
       resourceObject=null
       tiling=NoTile
       imageSize=<56 28>
       margin=<ltrb 0 0 0 0>
       outlineSettings {
        bUseBrushTransparency=false
        roundingType=FixedRadius
        width=0
        color {colorUseRule=UseColor_Specified, specifiedColor=<rgba 0 0 0 0>}
        cornerRadii=<4 4 4 4>
       }
       tintColor {colorUseRule=UseColor_ColorTable, specifiedColor=<rgba 0 0 0 0>}
       uVRegion {bIsValid=false, max=<0 0>, min=<0 0>}
      }
     }
    }
   }
   slot {class=VerticalBoxSlot}
  }
  Border_0 {
   widgetClassPath=&Border
   parent=LiteralWidgets_Box
   widget=Border_0
   properties {
    horizontalAlignment=HAlign_Left
    verticalAlignment=VAlign_Top
    background {
     bIsDynamicallyLoaded=false
     drawAs=RoundedBox
     imageType=NoImage
     mirroring=NoMirror
     resourceName=None
     resourceObject=null
     tiling=NoTile
     imageSize=<32 32>
     margin=<ltrb 0 0 0 0>
     outlineSettings {
      bUseBrushTransparency=false
      roundingType=FixedRadius
      width=0
      color {colorUseRule=UseColor_Specified, specifiedColor=<rgba 0.0075 0.0075 0.0075 1>}
      cornerRadii=<12 12 12 12>
     }
     tintColor {colorUseRule=UseColor_Specified, specifiedColor=<rgba 1 1 1 1>}
     uVRegion {bIsValid=false, max=<0 0>, min=<0 0>}
    }
    brushColor=<rgba 0.0212 0.0212 0.0212 1>
    padding=<ltrb 16 16 16 16>
   }
   slot {class=WrapBoxSlot, properties {verticalAlignment=VAlign_Top}}
  }
  VerticalBox_1 {
   widgetClassPath=&VerticalBox
   parent=Border_0
   widget=VerticalBox_1
   slot {class=BorderSlot, properties {horizontalAlignment=HAlign_Left, verticalAlignment=VAlign_Top, padding=<ltrb 16 16 16 16>}}
  }
  CommonTextBlock_272 {
   widgetClassPath=&CommonTextBlock
   parent=VerticalBox_1
   widget=CommonTextBlock_272
   properties {
    text=Random Pitch
    colorAndOpacity {colorUseRule=UseColor_Specified, specifiedColor=<rgba 0.5271 0.5271 0.5271 1>}
    font {
     bForceMonospaced=false
     bMaterialIsStencil=false
     fontMaterial=null
     fontObject=&Roboto
     letterSpacing=0
     monospacedWidth=1
     size=9
     skewAmount=0
     typefaceFontName=Regular
     outlineSettings {
      bApplyOutlineToDropShadows=false
      bMiteredCorners=false
      bSeparateFillAlpha=false
      outlineMaterial=null
      outlineSize=0
      outlineColor=<rgba 0 0 0 1>
     }
    }
    shadowOffset=<0 0>
   }
   slot {class=VerticalBoxSlot, properties {horizontalAlignment=HAlign_Center, padding=<ltrb 8 8 8 0>}}
  }
  WrapBox {
   widgetClassPath=&WrapBox
   bIsVariable=true
   parent=VerticalBox_1
   widget=WrapBox
   properties {innerSlotPadding=<16 0>}
   slot {class=VerticalBoxSlot, properties {horizontalAlignment=HAlign_Left, verticalAlignment=VAlign_Top, padding=<ltrb 8 0 8 0>}}
  }
  RandomPitchBase_Knob {
   widgetClassPath=/TechAudioToolsContent/Widgets/MetaSoundLiterals/Float/EUW_MetaSoundFloat_Knob_Default
   bIsVariable=true
   parent=WrapBox
   widget=RandomPitchBase_Knob
   properties {
    input Name=Random.PitchShift.Base
    label Text=Base
    metaSoundLiteralViewModel_Float=&self:WidgetTree.RandomPitchBase_Knob.MetaSoundLiteralViewModel_Float_0
    knob Style {
     material=null
     desiredSize=<96 96>
     knobAccentColor=<rgba 0.06 0.06 0.06 1>
     knobBarColor=<rgba 0.067 0.067 0.067 1>
     knobBarFillMaxColor=<rgba 0.8707 1 0 1>
     knobBarFillMidColor=<rgba 0.3169 1 0 1>
     knobBarFillMinColor=<rgba 0 1 0.0698 1>
     knobBarFillTintColor=<rgba 0.96 0.96 0.96 1>
     knobBarShadowColor=<rgba 0.067 0.067 0.067 1>
     knobEdgeFillColor=<rgba 0.0156 0.0156 0.0156 1>
     knobIndicatorDotColor=<rgba 0 1 0.0698 1>
     knobMainColor=<rgba 0.1406 0.1406 0.1406 1>
     knobShadowColor=<rgba 0.06 0.06 0.06 1>
     knobSmoothBevelColor=<rgba 0.0417 0.0417 0.0417 1>
     textBoxStyle {
      backgroundColor {colorUseRule=UseColor_ColorTable, specifiedColor=<rgba 0 0 0 0>}
      backgroundImage {
       bIsDynamicallyLoaded=false
       drawAs=RoundedBox
       imageType=NoImage
       mirroring=NoMirror
       resourceName=None
       resourceObject=null
       tiling=NoTile
       imageSize=<56 28>
       margin=<ltrb 0 0 0 0>
       outlineSettings {
        bUseBrushTransparency=false
        roundingType=FixedRadius
        width=0
        color {colorUseRule=UseColor_Specified, specifiedColor=<rgba 0 0 0 0>}
        cornerRadii=<4 4 4 4>
       }
       tintColor {colorUseRule=UseColor_ColorTable, specifiedColor=<rgba 0 0 0 0>}
       uVRegion {bIsValid=false, max=<0 0>, min=<0 0>}
      }
     }
    }
   }
   slot {class=WrapBoxSlot}
  }
  RandomPitchSpread_Knob {
   widgetClassPath=&EUW_MetaSoundFloat_Knob_Default
   bIsVariable=true
   parent=WrapBox
   widget=RandomPitchSpread_Knob
   properties {
    default Value=0.5
    input Name=Random.PitchShift.Spread
    label Text=Spread
    metaSoundLiteralViewModel_Float=&self:WidgetTree.RandomPitchSpread_Knob.MetaSoundLiteralViewModel_Float_0
    knob Style {
     material=null
     desiredSize=<96 96>
     knobAccentColor=<rgba 0.06 0.06 0.06 1>
     knobBarColor=<rgba 0.067 0.067 0.067 1>
     knobBarFillMaxColor=<rgba 0.8707 1 0 1>
     knobBarFillMidColor=<rgba 0.3169 1 0 1>
     knobBarFillMinColor=<rgba 0 1 0.0698 1>
     knobBarFillTintColor=<rgba 0.96 0.96 0.96 1>
     knobBarShadowColor=<rgba 0.067 0.067 0.067 1>
     knobEdgeFillColor=<rgba 0.0156 0.0156 0.0156 1>
     knobIndicatorDotColor=<rgba 0 1 0.0698 1>
     knobMainColor=<rgba 0.1406 0.1406 0.1406 1>
     knobShadowColor=<rgba 0.06 0.06 0.06 1>
     knobSmoothBevelColor=<rgba 0.0417 0.0417 0.0417 1>
     textBoxStyle {
      backgroundColor {colorUseRule=UseColor_ColorTable, specifiedColor=<rgba 0 0 0 0>}
      backgroundImage {
       bIsDynamicallyLoaded=false
       drawAs=RoundedBox
       imageType=NoImage
       mirroring=NoMirror
       resourceName=None
       resourceObject=null
       tiling=NoTile
       imageSize=<56 28>
       margin=<ltrb 0 0 0 0>
       outlineSettings {
        bUseBrushTransparency=false
        roundingType=FixedRadius
        width=0
        color {colorUseRule=UseColor_Specified, specifiedColor=<rgba 0 0 0 0>}
        cornerRadii=<4 4 4 4>
       }
       tintColor {colorUseRule=UseColor_ColorTable, specifiedColor=<rgba 0 0 0 0>}
       uVRegion {bIsValid=false, max=<0 0>, min=<0 0>}
      }
     }
    }
   }
   slot {class=WrapBoxSlot}
  }
  CommonTextBlock {
   widgetClassPath=&CommonTextBlock
   parent=VerticalBox_1
   widget=CommonTextBlock
   properties {
    text=Semitones
    colorAndOpacity {colorUseRule=UseColor_Specified, specifiedColor=<rgba 0.5271 0.5271 0.5271 1>}
    font {
     bForceMonospaced=false
     bMaterialIsStencil=false
     fontMaterial=null
     fontObject=&Roboto
     letterSpacing=0
     monospacedWidth=1
     size=9
     skewAmount=0
     typefaceFontName=Regular
     outlineSettings {
      bApplyOutlineToDropShadows=false
      bMiteredCorners=false
      bSeparateFillAlpha=false
      outlineMaterial=null
      outlineSize=0
      outlineColor=<rgba 0 0 0 1>
     }
    }
    shadowOffset=<0 0>
   }
   slot {class=VerticalBoxSlot, properties {horizontalAlignment=HAlign_Center, padding=<ltrb 8 0 8 8>}}
  }
 ]
}
